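# Animation Blueprint for the wingsuit gadget skeleton that plays the jetpack's cooldown animations.
# One state machine cycles from a reference-pose Base through Cooldown Start, Loop and End, with zero-duration crossfades.
# bIsCooldown gates entry and, when false, exit from Loop; Start and End advance automatically when their sequences finish.
# The Loop sequence has looping disabled, and the empty event graph never sets bIsCooldown, which is not declared here.
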AnimBlueprint /WingSuit/Animation/Game/MainPlayer/AnimAssets/JetpackAnim/JetPack_Gadget_AnimBP
target_skeleton=/WingSuit/Assets/MeshGadget/Wingsuit_Gadget_Skeleton
graphs [
 AnimGraph {nodes [
   AnimGraphNode_StateMachine_0 = Cooldown SM() {is=title AnimGraphNode_StateMachine}
   Output Pose(Result=@AnimGraphNode_StateMachine_0)
  ]}
 Cooldown SM {
  entry=Base Pose
  path=AnimGraph.AnimGraphNode_StateMachine_0.Cooldown SM
  ids=id
  nodes [
   AnimStateNode_0 = Base Pose() {is=title AnimStateNode}
   AnimStateNode_1 = Cooldown Start() {is=title AnimStateNode}
   AnimStateNode_2 = Cooldown Loop() {is=title AnimStateNode}
   AnimStateNode_3 = Cooldown End() {is=title AnimStateNode}
   AnimStateTransitionNode_0 = AnimStateTransitionNode() {bAllowInertializationForSelfTransitions=true, crossfadeDuration=0}
   AnimStateTransitionNode_1 = AnimStateTransitionNode() {bAllowInertializationForSelfTransitions=true, bAutomaticRuleBasedOnSequencePlayerInState=true, crossfadeDuration=0}
   AnimStateTransitionNode_2 = AnimStateTransitionNode() {bAllowInertializationForSelfTransitions=true, crossfadeDuration=0}
   AnimStateTransitionNode_3 = AnimStateTransitionNode() {bAllowInertializationForSelfTransitions=true, bAutomaticRuleBasedOnSequencePlayerInState=true, crossfadeDuration=0}
  ]
  links [
   {
    dst=Cooldown Start
    rule=AnimGraph.AnimGraphNode_StateMachine_0.Cooldown SM.AnimStateTransitionNode_0.Transition
    src=Base Pose
    via=AnimStateTransitionNode_0
   }
   {
    dst=Base Pose
    rule=AnimGraph.AnimGraphNode_StateMachine_0.Cooldown SM.AnimStateTransitionNode_3.Transition
    src=Cooldown End
    via=AnimStateTransitionNode_3
   }
   {
    dst=Cooldown End
    rule=AnimGraph.AnimGraphNode_StateMachine_0.Cooldown SM.AnimStateTransitionNode_2.Transition
    src=Cooldown Loop
    via=AnimStateTransitionNode_2
   }
   {
    dst=Cooldown Loop
    rule=AnimGraph.AnimGraphNode_StateMachine_0.Cooldown SM.AnimStateTransitionNode_1.Transition
    src=Cooldown Start
    via=AnimStateTransitionNode_1
   }
  ]
 }
 Base Pose {
  path=AnimGraph.AnimGraphNode_StateMachine_0.Cooldown SM.AnimStateNode_0.Base Pose
  nodes [Output Animation Pose(Result=Local Space Ref Pose())]
 }
 Cooldown Start {
  path=AnimGraph.AnimGraphNode_StateMachine_0.Cooldown SM.AnimStateNode_1.Cooldown Start
  nodes [Output Animation Pose(Result=Sequence Player 'JetPack_cooldownStart_Wingsuit_Gadget'() {bLoopAnimation=false, sequence=&/JetPack_cooldownStart_Wingsuit_Gadget})]
 }
 Cooldown Loop {
  path=AnimGraph.AnimGraphNode_StateMachine_0.Cooldown SM.AnimStateNode_2.Cooldown Loop
  nodes [Output Animation Pose(Result=Sequence Player 'JetPack_cooldownLoop_Wingsuit_Gadget'() {bLoopAnimation=false, sequence=&/JetPack_cooldownLoop_Wingsuit_Gadget})]
 }
 Cooldown End {
  path=AnimGraph.AnimGraphNode_StateMachine_0.Cooldown SM.AnimStateNode_3.Cooldown End
  nodes [Output Animation Pose(Result=Sequence Player 'JetPack_cooldownEnd_Wingsuit_Gadget'() {bLoopAnimation=false, sequence=&/JetPack_cooldownEnd_Wingsuit_Gadget})]
 }
 Transition {
  path=AnimGraph.AnimGraphNode_StateMachine_0.Cooldown SM.AnimStateTransitionNode_0.Transition
  nodes [Result(bCanEnterTransition=Get bIsCooldown() -> (bIsCooldown:bool))]
 }
 Transition {path=AnimGraph.AnimGraphNode_StateMachine_0.Cooldown SM.AnimStateTransitionNode_1.Transition, nodes [Result()]}
 Transition {
  path=AnimGraph.AnimGraphNode_StateMachine_0.Cooldown SM.AnimStateTransitionNode_2.Transition
  nodes [Result(bCanEnterTransition=NOT Boolean(A=Get bIsCooldown()))]
 }
 Transition {path=AnimGraph.AnimGraphNode_StateMachine_0.Cooldown SM.AnimStateTransitionNode_3.Transition, nodes [Result()]}
 EventGraph {links [], nodes []}
]
variables []
